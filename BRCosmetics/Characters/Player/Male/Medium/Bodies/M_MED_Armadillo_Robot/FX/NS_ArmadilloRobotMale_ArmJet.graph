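# Niagara effect for the arm jet flames on the male Armadillo Robot player character cosmetic.
# Its single JetPackFlames emitter bursts two mesh particles, split by spawn group to opposite X offsets.
# Their X scale combines an ignition overshoot curve, a sine flicker with a slightly different period per group, and an end fade.
# Three FN_MeshRenderer_v2 modules give each particle layered meshes with separate color, scale and orientation.
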NiagaraSystem /BRCosmetics/Characters/Player/Male/Medium/Bodies/M_MED_Armadillo_Robot/FX/NS_ArmadilloRobotMale_ArmJet
dependencies {
 usedDataInterfaces [&NiagaraDataInterfaceCurve]
 usedDynamicInputs [&UniformRangedFloat]
 usedModules [&SystemState]
}
emitters {JetPackFlames {
  emitterUpdateScript [
   &EmitterState()
   &SpawnBurst_Instantaneous(Spawn Count=2, Use Spawn Probability=false)
  ]
  particleSpawnScript [
   &FN_SetSpawnGroup(Number of Spawn Groups=2)
   &InitializeParticle(
    Color Mode=::NewEnumerator0
    Lifetime=3
    Mesh Scale=<2 1 1>
    Mesh Scale Mode=::NewEnumerator4
    Position Offset=&MaskVectorBySpawnGroup(Masked Vector=<2 0 0>, Passthrough Vector=<-2 0 0>)
    UsePositionOffset=true
    Write Material Random=true
   ) {module=InitializeParticle001}
  ]
  particleUpdateScript [
   &ParticleState()
   &ScaleMeshSize(Scale Factor=&MakeVector(
     X=&Multiply_Float(
      A=&Multiply_Float(
       A=&FloatFromCurve(
        FloatCurve=DataInterface() {propertyValues {Curve {keys [
            {arriveTangent=9.0725, interpMode=RCIM_Cubic, leaveTangent=9.0725, tangentMode=RCTM_User, time=0, value=0}
            {arriveTangent=1.6911, interpMode=RCIM_Cubic, leaveTangent=1.6911, tangentMode=RCTM_User, time=0.0762, value=5}
            {arriveTangentWeight=0.0866, interpMode=RCIM_Cubic, leaveTangentWeight=0.0774, tangentMode=RCTM_User, time=0.3, value=1}
            {arriveTangent=-0.0405, interpMode=RCIM_Cubic, leaveTangent=-0.0405, tangentMode=RCTM_User, time=1, value=1}
           ]}}}
        CurveIndex=$Emitter.Age
        Scale Curve=2
       )
       B=&Sine(
        Normalized Angle=$Emitter.Age
        Period=&MaskFloatBySpawnGroup(Masked Float=0.12, Passthrough Float=0.135)
        Scale=0.2
        Bias=0.5
       )
      )
      B=&FloatFromCurve(FloatCurve=DataInterface() {propertyValues {Curve {keys [
           {interpMode=RCIM_Cubic, tangentMode=RCTM_User, time=0, value=1}
           {interpMode=RCIM_Cubic, tangentMode=RCTM_User, time=0.2, value=1}
           <at 0.8 1>
           {tangentMode=RCTM_User, time=1, value=0}
          ]}}})
     )
     Y=1.5
     Z=1.5
    ))
   /Game/Effects/Niagara/Modules/Mesh/FN_MeshRenderer_v3/FN_MeshRenderer_v2(
    Color:LinearColor=<rgba 15 15 15 1>
    Create COlor Attribute=true
    Create Orientation Attribute=true
    Create Scale Attribute=true
    Orientation:Quat4f=&MultiplyQuaternion(
     Quaternion A=$Particles.MeshOrientation
     Quaternion B=&FindQuatBetweenTwoVectors(From Vector=<-1 0 0>, To Vector=<0 0 -1>)
    )
    Scale:Vector3f=&Multiply_Vector(A=<0.04 0.03 0.03>, B=$Particles.Scale)
   )
   &FN_MeshRenderer_v2(
    Color:LinearColor=<rgba 2 0 0.5 0.35>
    Create COlor Attribute=true
    Create Orientation Attribute=true
    Create Scale Attribute=true
    Orientation:Quat4f=&MultiplyQuaternion(
     Quaternion A=$Particles.MeshOrientation
     Quaternion B=&FindQuatBetweenTwoVectors(From Vector=<-1 0 0>, To Vector=<0 0 1>)
    )
    Scale:Vector3f=&Multiply_Vector(A=<0.0985 0.1 0.1>, B=$Particles.Scale)
   ) {module=FN_MeshRenderer_v2001}
   &FN_MeshRenderer_v2(
    Color:LinearColor=<rgba 0.5 0 0.03 8>
    Create COlor Attribute=true
    Create Scale Attribute=true
    Scale:Vector3f=&Multiply_Vector(A=<0.09 0.085 0.085>, B=$Particles.Scale)
   ) {module=FN_MeshRenderer_v2002}
  ]
  rendererClasses [&NiagaraMeshRendererProperties]
  renderers [
   {rendererClass=&NiagaraMeshRendererProperties, rendererIndex=0}
   {rendererClass=&NiagaraMeshRendererProperties, rendererIndex=1}
   {rendererClass=&NiagaraMeshRendererProperties, rendererIndex=2}
  ]
 }}
summary {
 systemName=NS_ArmadilloRobotMale_ArmJet
 emitters [{emitterName=JetPackFlames, rendererClasses [&NiagaraMeshRendererProperties]}]
 userVariables [User.bIsFrontEndPreview {type=NiagaraBool}]
}
