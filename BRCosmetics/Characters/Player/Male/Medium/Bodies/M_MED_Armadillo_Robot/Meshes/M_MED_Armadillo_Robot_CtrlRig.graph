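# Control Rig for the Armadillo Robot player body, adding procedural secondary motion to its armor bones.
# Parent constraints drive foot-ball and collar bones, then additive offsets weighted by remapped clavicle and head angles adjust the collar.
# Kalman-smoothed Ads_Float slides and shrinks three right-forearm fins in staggered, eased stages.
# Clavicle/head test controls are bypassed, debug draws remain enabled, and alternative constraint chains sit unconnected to execution.
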ControlRigBlueprint /BRCosmetics/Characters/Player/Male/Medium/Bodies/M_MED_Armadillo_Robot/Meshes/M_MED_Armadillo_Robot_CtrlRig
graphs [
 RigVMFunctionLibrary {links [], nodes []}
 RigVMModel {nodes [
   RigUnit_BeginExecution = &RigUnit_BeginExecution()
   n1 = /Script/ControlRig.RigUnit_Control_StaticMesh(
    Base=(Rotation=(X=0,Y=0,Z=0,W=1),Translation=(X=0,Y=0,Z=0),Scale3D=(X=1,Y=1,Z=1))
    Filter=(TranslationFilter=(bX=True,bY=True,bZ=True),RotationFilter=(bX=True,bY=True,bZ=True),ScaleFilter=(bX=True,bY=True,bZ=True))
    MeshTransform=(Rotation=(X=0,Y=0,Z=0,W=1),Translation=(X=0,Y=0,Z=20),Scale3D=(X=0.25,Y=0.25,Z=0.25))
    InitTransform=/Script/ControlRig.RigUnit_GetInitialBoneTransform(Bone=clavicle_l, Space=GlobalSpace) {is=struct RigVMUnitNode}.Transform
   ) {is=struct RigVMUnitNode}
   n2 = &RigUnit_GetBoneTransform(Bone=spine_05)
   /Script/ControlRig.RigUnit_SetBoneTransform(
    Bone=clavicle_l
    Space=LocalSpace
    Weight=1
    bPropagateToChildren=True
    Transform=/Script/RigVM.RigVMFunction_MathTransformSelectBool(
     Condition=true
     IfTrue=&RigUnit_GetBoneTransform(Bone=clavicle_l, Space=LocalSpace).Transform
     IfFalse=&RigVMFunction_MathTransformMakeRelative(Global=@n1.Result, Parent=@n2.Transform).Local
    ) {is=struct RigVMUnitNode}.Result
    ExecutePin=@RigUnit_BeginExecution.ExecutePin
   ) {is=struct RigVMUnitNode}
   n3 = &RigUnit_Control_StaticMesh(
    Base=(Rotation=(X=0,Y=0,Z=0,W=1),Translation=(X=0,Y=0,Z=0),Scale3D=(X=1,Y=1,Z=1))
    Filter=(TranslationFilter=(bX=True,bY=True,bZ=True),RotationFilter=(bX=True,bY=True,bZ=True),ScaleFilter=(bX=True,bY=True,bZ=True))
    MeshTransform=(Rotation=(X=0,Y=0,Z=0,W=1),Translation=(X=0,Y=0,Z=20),Scale3D=(X=0.25,Y=0.25,Z=0.25))
    InitTransform=&RigUnit_GetInitialBoneTransform(Bone=head, Space=GlobalSpace).Transform
   )
   n4 = &RigUnit_GetBoneTransform(Bone=neck_02)
   > &RigUnit_SetBoneTransform(
    Bone=head
    Space=LocalSpace
    Weight=1
    bPropagateToChildren=True
    Transform=&RigVMFunction_MathTransformSelectBool(
     Condition=true
     IfTrue=&RigUnit_GetBoneTransform(Bone=head, Space=LocalSpace).Transform
     IfFalse=&RigVMFunction_MathTransformMakeRelative(Global=@n3.Result, Parent=@n4.Transform).Local
    ).Result
   )
   > &RigUnit_ParentConstraint(
    AdvancedSettings=(InterpolationType=Average,RotationOrderForFilter=XZY)
    Child=(Type=Bone,Name="dyn_ball_main_l")
    Parents=((Item=(Type=Bone,Name="ball_l"),Weight=1))
    Weight=1
   )
   > &RigUnit_ParentConstraint(
    AdvancedSettings=(InterpolationType=Average,RotationOrderForFilter=XZY)
    Child=(Type=Bone,Name="dyn_ball_main_r")
    Parents=((Item=(Type=Bone,Name="ball_r"),Weight=1))
    Weight=1
   )
   > &RigUnit_ParentConstraint(
    AdvancedSettings=(InterpolationType=Average,RotationOrderForFilter=XZY)
    Child=(Type=Bone,Name="cr_collar_frt_clavicle")
    Parents=((Item=(Type=Bone,Name="spine_05"),Weight=1),(Item=(Type=Bone,Name="clavicle_l"),Weight=1))
    Weight=1
   )
   > &RigUnit_ParentConstraint(
    AdvancedSettings=(InterpolationType=Average,RotationOrderForFilter=XZY)
    Child=(Type=Bone,Name="parent_dyn_collar_mid_01")
    Parents=((Item=(Type=Bone,Name="spine_05"),Weight=0.5),(Item=(Type=Bone,Name="clavicle_l"),Weight=0.5))
    Weight=1
   )
   > &RigUnit_ParentConstraint(
    AdvancedSettings=(InterpolationType=Average,RotationOrderForFilter=XZY)
    Child=(Type=Bone,Name="cr_collar_frt_spine")
    Parents=((Item=(Type=Bone,Name="cr_collar_frt_clavicle"),Weight=1))
    Weight=1
   )
   n5 = &RigUnit_GetRelativeBoneTransform(Bone=clavicle_l, Space=spine_05)
   > /Script/ControlRig.RigUnit_ModifyBoneTransforms(
    BoneToModify=((Bone="cr_collar_frt_clavicle",Transform=(Rotation=(X=0,Y=0,Z=0,W=1),Translation=(X=0,Y=3,Z=0),Scale3D=(X=1,Y=1,Z=1))),(Bone="parent_dyn_collar_mid_01",Transform=(Rotation=(X=0,Y=0,Z=0,W=1),Translation=(X=0,Y=3,Z=0),Scale3D=(X=1,Y=1,Z=1))))
    Mode=AdditiveLocal
    WeightMaximum=1
    WeightMinimum=0
    Weight=&RigVMFunction_MathFloatRemap(
     SourceMaximum=45
     SourceMinimum=89
     TargetMinimum=0
     bClamp=true
     Value=&RigVMFunction_MathQuaternionToEuler(RotationOrder=ZXY, Value=@n5.Transform.Rotation).Result.Y
    ).Result
   ) {is=struct RigVMUnitNode}
   > /Script/ControlRig.RigUnit_DebugTransformMutable(
    Color=(R=1,G=0,B=0,A=1)
    DebugDrawSettings=(DepthPriority=SDPG_Foreground,Lifetime=-1)
    Mode=Axes
    Scale=10
    Space=None
    Thickness=0.6
    WorldOffset=(Rotation=(X=0,Y=0,Z=0,W=1),Translation=(X=0,Y=0,Z=0),Scale3D=(X=1,Y=1,Z=1))
    bEnabled=True
    Transform=&RigUnit_GetBoneTransform(Bone=cr_collar_frt_clavicle).Transform
   ) {is=struct RigVMUnitNode}
   n6 = &RigVMFunction_MathQuaternionToEuler(RotationOrder=XYZ, Value=&RigUnit_GetRelativeBoneTransform(Bone=head, Space=spine_05).Transform.Rotation)
   > &RigUnit_ModifyBoneTransforms(
    BoneToModify=((Bone="cr_collar_frt_spine",Transform=(Rotation=(X=0,Y=0,Z=0,W=1),Translation=(X=0,Y=0,Z=0),Scale3D=(X=1,Y=1,Z=1))),(Bone="driver_dyn_collar_mid_01",Transform=(Rotation=(X=0,Y=0,Z=0,W=1),Translation=(X=0,Y=0,Z=0),Scale3D=(X=1,Y=1,Z=1))))
    Mode=AdditiveLocal
    WeightMaximum=1
    WeightMinimum=0
    Weight=/Script/RigVM.RigVMFunction_MathFloatMax(
     A=&RigVMFunction_MathFloatRemap(SourceMaximum=50, SourceMinimum=0, TargetMinimum=0, bClamp=true, Value=@n6.Result.Y).Result
     B=&RigVMFunction_MathFloatRemap(SourceMaximum=-50, SourceMinimum=0, TargetMinimum=0, bClamp=true, Value=@n6.Result.X).Result
    ) {is=struct RigVMUnitNode, template="Maximum::Execute(in A,in B,out Result)"}.Result
    BoneToModify.0.Transform.Rotation=&RigVMFunction_MathQuaternionFromEuler(Euler=(X=0,Y=0,Z=20)).Result
    BoneToModify.1.Transform.Rotation=&RigVMFunction_MathQuaternionFromEuler(Euler=(X=0,Y=0,Z=15)).Result
   )
   > &RigUnit_DebugTransformMutable(
    Color=(R=1,G=0,B=0,A=1)
    DebugDrawSettings=(DepthPriority=SDPG_Foreground,Lifetime=-1)
    Mode=Box
    Scale=10
    Space=None
    Thickness=0.5
    WorldOffset=(Rotation=(X=0,Y=0,Z=0,W=1),Translation=(X=0,Y=0,Z=0),Scale3D=(X=1,Y=1,Z=1))
    bEnabled=True
    Transform=&RigUnit_GetBoneTransform(Bone=cr_collar_frt_spine, Space=LocalSpace).Transform
   )
   > &RigUnit_ParentConstraint(
    AdvancedSettings=(InterpolationType=Average,RotationOrderForFilter=XZY)
    Child=(Type=Bone,Name="parent_dyn_collar_frt_01")
    Parents=((Item=(Type=Bone,Name="cr_collar_frt_clavicle"),Weight=1))
    Weight=1
    bMaintainOffset=true
   )
   n7 = &RigVMFunction_KalmanFloat(BufferSize=13, Value=RigVMVariableNode() {variable=Ads_Float}.Value) {template="KalmanFilter::Execute(in Value,in BufferSize,out Result)"}
   > &RigUnit_ModifyBoneTransforms(
    BoneToModify=((Bone="dyn_dino_finA",Transform=(Rotation=(X=0,Y=0,Z=0,W=1),Translation=(X=-8,Y=0,Z=0),Scale3D=(X=1,Y=1,Z=1))))
    Mode=AdditiveLocal
    WeightMaximum=1
    WeightMinimum=0
    Weight=/Script/RigVM.RigVMFunction_AnimEasing(SourceMaximum=0.3, SourceMinimum=0, TargetMaximum=1, TargetMinimum=0, Type=CubicEaseInOut, Value=@n7.Result) {is=struct RigVMUnitNode}.Result
   )
   > &RigUnit_ModifyBoneTransforms(
    BoneToModify=((Bone="dyn_dino_finB",Transform=(Rotation=(X=0,Y=0,Z=0,W=1),Translation=(X=-8,Y=-3,Z=0),Scale3D=(X=1,Y=1,Z=1))))
    Mode=AdditiveLocal
    WeightMaximum=1
    WeightMinimum=0
    Weight=&RigVMFunction_AnimEasing(SourceMaximum=0.425, SourceMinimum=0.15, TargetMaximum=1, TargetMinimum=0, Type=CubicEaseInOut, Value=@n7.Result).Result
   )
   > &RigUnit_ModifyBoneTransforms(
    BoneToModify=((Bone="dyn_dino_finC",Transform=(Rotation=(X=0,Y=0,Z=0,W=1),Translation=(X=-5,Y=0,Z=0),Scale3D=(X=1,Y=1,Z=1))))
    Mode=AdditiveLocal
    WeightMaximum=1
    WeightMinimum=0
    Weight=&RigVMFunction_AnimEasing(SourceMaximum=0.45, SourceMinimum=0.25, TargetMaximum=1, TargetMinimum=0, Type=CubicEaseInOut, Value=@n7.Result).Result
   )
   > &RigUnit_ModifyBoneTransforms(
    BoneToModify=((Bone="dyn_dino_finA",Transform=(Rotation=(X=0,Y=0,Z=0,W=1),Translation=(X=0,Y=0,Z=0),Scale3D=(X=0,Y=0,Z=0))))
    Mode=AdditiveLocal
    WeightMaximum=1
    WeightMinimum=0
    Weight=&RigVMFunction_AnimEasing(SourceMaximum=0.35, SourceMinimum=0.25, TargetMaximum=1, TargetMinimum=0, Type=Linear, Value=@n7.Result).Result
   )
   > &RigUnit_ModifyBoneTransforms(
    BoneToModify=((Bone="dyn_dino_finB",Transform=(Rotation=(X=0,Y=0,Z=0,W=1),Translation=(X=0,Y=0,Z=0),Scale3D=(X=0,Y=0,Z=0))))
    Mode=AdditiveLocal
    WeightMaximum=1
    WeightMinimum=0
    Weight=&RigVMFunction_AnimEasing(SourceMaximum=0.45, SourceMinimum=0.35, TargetMaximum=1, TargetMinimum=0, Type=Linear, Value=@n7.Result).Result
   )
   > &RigUnit_ModifyBoneTransforms(
    BoneToModify=((Bone="dyn_dino_finC",Transform=(Rotation=(X=0,Y=0,Z=0,W=1),Translation=(X=0,Y=0,Z=0),Scale3D=(X=0,Y=0,Z=0))))
    Mode=AdditiveLocal
    WeightMaximum=1
    WeightMinimum=0
    Weight=&RigVMFunction_AnimEasing(SourceMaximum=0.6, SourceMinimum=0.5, TargetMaximum=1, TargetMinimum=0, Type=Linear, Value=@n7.Result).Result
   )
   n8 = &RigUnit_GetRelativeBoneTransform(Bone=head, Space=spine_05)
   > &RigUnit_ModifyBoneTransforms(
    BoneToModify=((Bone="parent_dyn_collar_bck_01",Transform=(Rotation=(X=0,Y=0,Z=0,W=1),Translation=(X=0,Y=1,Z=2),Scale3D=(X=1,Y=1,Z=1))),(Bone="parent_dyn_collar_mid_01",Transform=(Rotation=(X=0,Y=0,Z=0,W=1),Translation=(X=0,Y=1.5,Z=0),Scale3D=(X=1,Y=1,Z=1))))
    Mode=AdditiveLocal
    WeightMaximum=1
    WeightMinimum=0
    Weight=&RigVMFunction_MathFloatRemap(
     SourceMaximum=-50
     SourceMinimum=0
     TargetMinimum=0
     bClamp=False
     Value=&RigVMFunction_MathQuaternionToEuler(RotationOrder=XYZ, Value=@n8.Transform.Rotation).Result.Z
    ).Result
    BoneToModify.0.Transform.Rotation=&RigVMFunction_MathQuaternionFromEuler(Euler=(X=0,Y=10,Z=40)).Result
    BoneToModify.1.Transform.Rotation=&RigVMFunction_MathQuaternionFromEuler(Euler=(X=0,Y=30,Z=30)).Result
   )
   n9 = &RigUnit_GetInitialBoneTransform(Bone=dyn_ball_main_l, Space=GlobalSpace)
   n10 = &RigUnit_GetInitialBoneTransform(Bone=ball_l, Space=GlobalSpace)
   /Script/ControlRig.RigUnit_SetBoneRotation(
    Bone=dyn_ball_main_l
    ExecutePin=()
    Space=GlobalSpace
    Weight=1
    bPropagateToChildren=True
    Rotation=&RigVMFunction_MathTransformMakeRelative(
     Global=&RigUnit_GetBoneTransform(Bone=ball_l).Transform
     Parent=&RigVMFunction_MathTransformMakeRelative(Parent=@n9.Transform, Global=@n10.Transform).Local
    ).Local.Rotation
   ) {is=struct RigVMUnitNode}
   n11 = &RigUnit_GetInitialBoneTransform(Bone=dyn_ball_main_r, Space=GlobalSpace)
   n12 = &RigUnit_GetInitialBoneTransform(Bone=ball_r, Space=GlobalSpace)
   > &RigUnit_SetBoneRotation(
    Bone=dyn_ball_main_r
    Space=GlobalSpace
    Weight=1
    bPropagateToChildren=True
    Rotation=&RigVMFunction_MathTransformMakeRelative(
     Global=&RigUnit_GetBoneTransform(Bone=ball_r).Transform
     Parent=&RigVMFunction_MathTransformMakeRelative(Parent=@n11.Transform, Global=@n12.Transform).Local
    ).Local.Rotation
   )
   /Script/ControlRig.RigUnit_TransformConstraint(
    BaseBone=None
    BaseTransform=(Rotation=(X=0,Y=0,Z=0,W=1),Translation=(X=0,Y=0,Z=0),Scale3D=(X=1,Y=1,Z=1))
    BaseTransformSpace=GlobalSpace
    Bone=cr_collar_frt_clavicle
    ExecutePin=()
    Targets="((Transform=(Rotation=(X=0,Y=0,Z=0,W=1),Translation=(X=0,Y=0,Z=0),Scale3D=(X=1,Y=1,Z=1)),Weight=0.5,bMaintainOffset=True,Filter=(TranslationFilter=(bX=True,bY=True,bZ=True),RotationFilter=(bX=True,bY=True,bZ=True),ScaleFilter=(bX=True,bY=True,bZ=True))),(T ...(truncated)"
    bUseInitialTransforms=True
    Targets.0.Transform=&RigUnit_GetBoneTransform(Bone=spine_05).Transform
    Targets.1.Transform=&RigUnit_GetBoneTransform(Bone=clavicle_l).Transform
   ) {is=struct RigVMUnitNode}
   > &RigUnit_TransformConstraint(
    BaseBone=None
    BaseTransform=(Rotation=(X=0,Y=0,Z=0,W=1),Translation=(X=0,Y=0,Z=0),Scale3D=(X=1,Y=1,Z=1))
    BaseTransformSpace=GlobalSpace
    Bone=parent_dyn_collar_mid_01
    Targets="((Transform=(Rotation=(X=0,Y=0,Z=0,W=1),Translation=(X=0,Y=0,Z=0),Scale3D=(X=1,Y=1,Z=1)),Weight=0.6,bMaintainOffset=True,Filter=(TranslationFilter=(bX=True,bY=True,bZ=True),RotationFilter=(bX=True,bY=True,bZ=True),ScaleFilter=(bX=True,bY=True,bZ=True))),(T ...(truncated)"
    bUseInitialTransforms=True
    Targets.0.Transform=&RigUnit_GetBoneTransform(Bone=spine_05).Transform
    Targets.1.Transform=&RigUnit_GetBoneTransform(Bone=clavicle_l).Transform
   )
   > &RigUnit_SetBoneTransform(
    Bone=cr_collar_frt_spine
    Space=GlobalSpace
    Weight=1
    bPropagateToChildren=True
    Transform=&RigUnit_GetBoneTransform(Bone=cr_collar_frt_clavicle).Transform
   )
   &RigUnit_TransformConstraint(
    BaseBone=None
    BaseTransform=(Rotation=(X=0,Y=0,Z=0,W=1),Translation=(X=0,Y=0,Z=0),Scale3D=(X=1,Y=1,Z=1))
    BaseTransformSpace=GlobalSpace
    Bone=parent_dyn_collar_frt_01
    ExecutePin=()
    Targets=((Transform=(Rotation=(X=0,Y=0,Z=0,W=1),Translation=(X=0,Y=0,Z=0),Scale3D=(X=1,Y=1,Z=1)),Weight=1,bMaintainOffset=True,Filter=(TranslationFilter=(bX=True,bY=True,bZ=True),RotationFilter=(bX=True,bY=True,bZ=True),ScaleFilter=(bX=True,bY=True,bZ=True))))
    bUseInitialTransforms=True
    Targets.0.Transform=&RigUnit_GetBoneTransform(Bone=cr_collar_frt_clavicle).Transform
   )
  ]}
]
hierarchy [
 root {loc=<0 0 0>, rot=<0 0 0 1>}
 pelvis {parent=root, loc=<0 -0.7071 93.1191>, rot=<0.0127 0.707 -0.0127 -0.707>}
 spine_01 {parent=pelvis, loc=<6.7494 -0.6388 0>, rot=<0 0 0.034 -0.9994>}
 spine_02 {parent=spine_01, loc=<7.2518 -0.035 0>, rot=<0 0 0.0238 -0.9997>}
 spine_03 {parent=spine_02, loc=<6.0104 0 0>, rot=<0 0 0.0149 -0.9999>}
 spine_04 {parent=spine_03, loc=<8.1024 0 0>, rot=<0 0 -0.0132 -0.9999>}
 spine_05 {parent=spine_04, loc=<10.0146 0 0>, rot=<0 0 -0.0015 -1>}
 clavicle_l {parent=spine_05, loc=<4.3575 -0.1466 -2.3223>, rot=<0.0465 -0.7289 -0.009 -0.6829>}
 upperarm_l {parent=clavicle_l, loc=<15.5111 0 0>, rot=<0.0226 -0.3234 -0.0088 -0.9459>}
 lowerarm_l {parent=upperarm_l, loc=<26.6514 0 0>, rot=<0.0391 -0.0099 -0.2453 -0.9686>}
 hand_l {parent=lowerarm_l, loc=<22.3491 0 0>, rot=<-0.7081 -0.0507 0.0975 -0.6975>}
 index_metacarpal_l {parent=hand_l, loc=<1.6535 0.0183 -2.96>, rot=<0.0158 -0.0124 -0.0173 -0.9996>}
 index_01_l {parent=index_metacarpal_l, loc=<8.3008 0.0357 -0.0524>, rot=<-0.0044 0.0485 0.0924 -0.9945>}
 index_02_l {parent=index_01_l, loc=<3.3587 -0.3004 0.0452>, rot=<-0.002 0.0027 0.3225 -0.9466>}
 index_03_l {parent=index_02_l, loc=<2.5578 -0.123 -0.0252>, rot=<-0.002 0.0067 0.2558 -0.9667>}
 middle_metacarpal_l {parent=hand_l, loc=<2.1856 -0.1725 -0.5984>, rot=<0.006 0.0168 -0.0329 -0.9993>}
 middle_01_l {parent=middle_metacarpal_l, loc=<7.3275 -0.2447 0.0129>, rot=<-0.0208 0.0042 0.0858 -0.9961>}
 middle_02_l {parent=middle_01_l, loc=<4.0123 0.1401 0.1175>, rot=<-0.0218 0.0258 0.3512 -0.9357>}
 middle_03_l {parent=middle_02_l, loc=<2.6077 0.0534 0.1373>, rot=<0.0082 0.0188 0.1708 -0.9851>}
 pinky_metacarpal_l {parent=hand_l, loc=<1.7616 -0.9753 3.3814>, rot=<-0.0397 0.0814 -0.069 -0.9935>}
 pinky_01_l {parent=pinky_metacarpal_l, loc=<7.1709 -0.4831 -0.1094>, rot=<-0.0654 -0.0035 0.2242 -0.9723>}
 pinky_02_l {parent=pinky_01_l, loc=<2.8274 0.0893 -0.1043>, rot=<0.0037 0.0011 0.2854 -0.9584>}
 pinky_03_l {parent=pinky_02_l, loc=<2.2046 0.0307 -0.0078>, rot=<-0.0002 0.0098 0.1445 -0.9895>}
 ring_metacarpal_l {parent=hand_l, loc=<2.2074 -0.3625 1.3638>, rot=<-0.0284 0.0722 -0.0458 -0.9959>}
 ring_01_l {parent=ring_metacarpal_l, loc=<7.0295 -0.6197 -0.1446>, rot=<-0.0489 -0.0252 0.1892 -0.9804>}
 ring_02_l {parent=ring_01_l, loc=<3.6174 0.4647 -0.1027>, rot=<0 0 0.2806 -0.9598>}
 ring_03_l {parent=ring_02_l, loc=<2.4152 0.114 0.0199>, rot=<0 0 0.1189 -0.9929>}
 thumb_01_l {parent=hand_l, loc=<1.3118 -0.3191 -2.459>, rot=<0.6546 -0.4042 0.0664 -0.6354>}
 thumb_02_l {parent=thumb_01_l, loc=<5.9336 -0.0448 0.0401>, rot=<-0.0037 0.0213 0.298 -0.9543>}
 thumb_03_l {parent=thumb_02_l, loc=<2.2223 0.3682 -0.0907>, rot=<0.0118 -0.013 0.0066 -0.9998>}
 lowerarm_twist_01_l {parent=lowerarm_l, loc=<17.8549 0 0>, rot=<0 0 0 1>}
 lowerarm_twist_02_l {parent=lowerarm_l, loc=<9.6199 0 0>, rot=<0 0 0 1>}
 dyn_lowerarm_offset_l {parent=lowerarm_l, loc=<0 0 0>, rot=<0 0 0 -1>}
 dyn_lowerarm_inside_l {parent=lowerarm_l, loc=<0 0 0>, rot=<0 0 0 -1>}
 dyn_shldr_main_l {parent=upperarm_l, loc=<0 0 0>, rot=<0 0 0 1>}
 dyn_shldr_base_l {parent=dyn_shldr_main_l, loc=<5.569 -0.5139 1.2508>, rot=<0.6799 0.174 0.1519 -0.696>}
 dyn_shldr_01_l {parent=dyn_shldr_base_l, loc=<11.3061 -3.6913 0>, rot=<0 0 -0.087 -0.9962>}
 dyn_shldr_02_l {parent=dyn_shldr_01_l, loc=<-16.5123 -1.0953 0>, rot=<0 0 0.0912 -0.9958>}
 dyn_upperarm_offset_l {parent=upperarm_l, loc=<0 0 0>, rot=<0 0 0 1>}
 upperarm_twist_01_l {parent=upperarm_l, loc=<0.0041 0 0>, rot=<0 0 0 1>}
 upperarm_twist_02_l {parent=upperarm_l, loc=<13.7941 0 0>, rot=<0 0 0 1>}
 clavicle_r {parent=spine_05, loc=<4.3575 -0.1466 2.3223>, rot=<0.7289 0.0465 -0.6829 0.009>}
 upperarm_r {parent=clavicle_r, loc=<-15.5111 0 0>, rot=<0.0226 -0.3234 -0.0088 -0.9459>}
 lowerarm_r {parent=upperarm_r, loc=<-26.6515 0 0>, rot=<0.0391 -0.0099 -0.2453 -0.9686>}
 hand_r {parent=lowerarm_r, loc=<-22.3492 0 0>, rot=<-0.7081 -0.0507 0.0975 -0.6975>}
 index_metacarpal_r {parent=hand_r, loc=<-1.6534 -0.0181 2.9601>, rot=<0.0158 -0.0124 -0.0173 -0.9996>}
 index_01_r {parent=index_metacarpal_r, loc=<-8.3008 -0.0357 0.0524>, rot=<-0.0044 0.0485 0.0924 -0.9945>}
 index_02_r {parent=index_01_r, loc=<-3.3589 0.3003 -0.0453>, rot=<-0.002 0.0027 0.3225 -0.9466>}
 index_03_r {parent=index_02_r, loc=<-2.5578 0.1231 0.0251>, rot=<-0.002 0.0067 0.2558 -0.9667>}
 middle_metacarpal_r {parent=hand_r, loc=<-2.1856 0.1726 0.5984>, rot=<0.006 0.0168 -0.0329 -0.9993>}
 middle_01_r {parent=middle_metacarpal_r, loc=<-7.3273 0.2447 -0.0129>, rot=<-0.0208 0.0042 0.0858 -0.9961>}
 middle_02_r {parent=middle_01_r, loc=<-4.0142 -0.1411 -0.07>, rot=<-0.0217 0.0259 0.3512 -0.9357>}
 middle_03_r {parent=middle_02_r, loc=<-2.6074 -0.0534 -0.1312>, rot=<0.0083 0.0188 0.1708 -0.9851>}
 pinky_metacarpal_r {parent=hand_r, loc=<-1.7617 0.9753 -3.3814>, rot=<-0.0397 0.0814 -0.069 -0.9935>}
 pinky_01_r {parent=pinky_metacarpal_r, loc=<-7.1707 0.4832 0.1094>, rot=<-0.0654 -0.0035 0.2242 -0.9723>}
 pinky_02_r {parent=pinky_01_r, loc=<-2.8274 -0.0892 0.1043>, rot=<0.0037 0.0011 0.2854 -0.9584>}
 pinky_03_r {parent=pinky_02_r, loc=<-2.2045 -0.031 0.0078>, rot=<-0.0002 0.0098 0.1445 -0.9895>}
 ring_metacarpal_r {parent=hand_r, loc=<-2.2073 0.3627 -1.3638>, rot=<-0.0284 0.0722 -0.0458 -0.9959>}
 ring_01_r {parent=ring_metacarpal_r, loc=<-7.0297 0.6197 0.1446>, rot=<-0.0489 -0.0252 0.1892 -0.9804>}
 ring_02_r {parent=ring_01_r, loc=<-3.6171 -0.4648 0.1027>, rot=<0 0 0.2806 -0.9598>}
 ring_03_r {parent=ring_02_r, loc=<-2.4153 -0.1141 -0.0199>, rot=<0 0 0.1189 -0.9929>}
 thumb_01_r {parent=hand_r, loc=<-1.3119 0.3192 2.459>, rot=<0.6546 -0.4042 0.0664 -0.6354>}
 thumb_02_r {parent=thumb_01_r, loc=<-5.9335 0.0447 -0.0401>, rot=<-0.0037 0.0213 0.298 -0.9543>}
 thumb_03_r {parent=thumb_02_r, loc=<-2.2223 -0.3682 0.0907>, rot=<0.0118 -0.013 0.0066 -0.9998>}
 lowerarm_twist_01_r {parent=lowerarm_r, loc=<-17.855 0 0>, rot=<0 0 0 1>}
 lowerarm_twist_02_r {parent=lowerarm_r, loc=<-9.6199 0 0>, rot=<0 0 0 1>}
 dyn_lowerarm_dyno_01_r {parent=lowerarm_r, loc=<-0.0001 0 0.0002>, rot=<0 0 0 1>}
 dyn_lowerarm_dyno_02_r {parent=dyn_lowerarm_dyno_01_r, loc=<-5.5449 0 -5.8369>, rot=<0 0 0 1>}
 dyn_dino_finA {parent=dyn_lowerarm_dyno_02_r, loc=<-9.7462 1.7524 -4.4449>, rot=<0.5287 0.5227 -0.4758 0.47>}
 dyn_dino_finB {parent=dyn_lowerarm_dyno_02_r, loc=<-20.1988 2.4395 -5.9774>, rot=<-0.5247 -0.5267 0.4793 -0.4664>}
 dyn_dino_finC {parent=dyn_lowerarm_dyno_02_r, loc=<-30.3033 1.8922 -1.9847>, rot=<-0.4003 -0.6253 0.5719 -0.3489>}
 dyn_shldr_main_r {parent=upperarm_r, loc=<-0.0001 0 0.0002>, rot=<0 0 0 1>}
 dyn_shldr_base_r {parent=dyn_shldr_main_r, loc=<-5.5687 0.5139 -1.2512>, rot=<0.6799 0.174 0.1519 -0.696>}
 dyn_shldr_01_r {parent=dyn_shldr_base_r, loc=<-11.3062 3.691 0>, rot=<0 0 -0.087 -0.9962>}
 dyn_shldr_02_r {parent=dyn_shldr_01_r, loc=<16.5122 1.095 0>, rot=<0 0 0.0912 -0.9958>}
 dyn_upperarm_offset_r {parent=upperarm_r, loc=<-0.0001 0 0.0002>, rot=<0 0 0 1>}
 upperarm_twist_01_r {parent=upperarm_r, loc=<-0.0041 0 0>, rot=<0 0 0 1>}
 upperarm_twist_02_r {parent=upperarm_r, loc=<-13.7942 0 0>, rot=<0 0 0 1>}
 neck_01 {parent=spine_05, loc=<11.3138 0.1529 0>, rot=<0 0 -0.2003 -0.9797>}
 neck_02 {parent=neck_01, loc=<4.5627 0 0>, rot=<0 0 0.0504 -0.9987>}
 head {parent=neck_02, loc=<5.3914 0 0>, rot=<0 0 0.115 -0.9934>}
 head_offset {parent=neck_02, loc=<2.6913 1.8435 0>, rot=<0 0 0.115 -0.9934>}
 parent_dyn_collar_frt_01 {parent=spine_05, loc=<11.9047 4.6093 -14.8805>, rot=<0.7874 -0.0074 -0.1199 -0.6047>}
 driver_dyn_collar_frt_01 {parent=parent_dyn_collar_frt_01, loc=<0 0 0>, rot=<0 0 0 1>}
 dyn_collar_frt_01 {parent=driver_dyn_collar_frt_01, loc=<0 0 0>, rot=<0 0 0 1>}
 dyn_collar_frt_02 {parent=dyn_collar_frt_01, loc=<3.9758 0 0>, rot=<0 0 -0.0129 -0.9999>}
 dyn_collar_frt_03 {parent=dyn_collar_frt_02, loc=<3.7697 0 0>, rot=<0 0 -0.0129 -0.9999>}
 parent_dyn_collar_mid_01 {parent=spine_05, loc=<13.2617 -3.6596 -11.5636>, rot=<0.8703 -0.0402 -0.06 -0.4872>}
 driver_dyn_collar_mid_01 {parent=parent_dyn_collar_mid_01, loc=<0 0 0>, rot=<0 0 0 1>}
 dyn_collar_mid_01 {parent=driver_dyn_collar_mid_01, loc=<0 0 0>, rot=<0 0 0 1>}
 dyn_collar_mid_02 {parent=dyn_collar_mid_01, loc=<3.9758 0 0>, rot=<0 0 0 1>}
 dyn_collar_mid_03 {parent=dyn_collar_mid_02, loc=<3.7768 0.0162 -0.09>, rot=<0 0 0 1>}
 parent_dyn_collar_bck_01 {parent=spine_05, loc=<13.2071 -9.9668 -3.4104>, rot=<0.9833 -0.0141 -0.0084 -0.1812>}
 driver_dyn_collar_bck_01 {parent=parent_dyn_collar_bck_01, loc=<0 0 0>, rot=<0 0 0 1>}
 dyn_collar_bck_01 {parent=driver_dyn_collar_bck_01, loc=<0 0 0>, rot=<0 0 0 1>}
 dyn_collar_bck_02 {parent=dyn_collar_bck_01, loc=<3.9758 0 0>, rot=<0 0 0 1>}
 dyn_collar_bck_03 {parent=dyn_collar_bck_02, loc=<3.7864 0.0495 -0.1303>, rot=<0 0 0 1>}
 cr_collar_frt_clavicle {parent=spine_05, loc=<11.9047 4.6093 -14.8805>, rot=<0.7874 -0.0074 -0.1199 -0.6047>}
 cr_collar_frt_spine {parent=spine_05, loc=<11.9047 4.6093 -14.8805>, rot=<0.7874 -0.0074 -0.1199 -0.6047>}
 thigh_l {parent=pelvis, loc=<-2.5693 -0.0109 -7.7501>, rot=<0.0012 0.0572 0.0155 -0.9982>}
 calf_l {parent=thigh_l, loc=<-39.0484 0 0>, rot=<0.0006 -0.0161 -0.0391 -0.9991>}
 calf_twist_02_l {parent=calf_l, loc=<-18.8064 -0.0967 0.0397>, rot=<0.0237 0.0095 -0.0035 -0.9997>}
 foot_l {parent=calf_l, loc=<-42.561 0.1077 0.4292>, rot=<0.0418 -0.0455 0.0343 -0.9975>}
 dyn_heel_offset_l {parent=foot_l, loc=<-0.08 0.3776 -0.1451>, rot=<-0.0284 0.0354 -0.7384 -0.6729>}
 dyn_heel_01_l {parent=dyn_heel_offset_l, loc=<-5.3597 7.3285 -0.0893>, rot=<0 0 0.0515 -0.9987>}
 dyn_heel_02_l {parent=dyn_heel_01_l, loc=<-4.6306 -7.0743 0.2556>, rot=<0.0002 -0.0005 0.5638 -0.8259>}
 dyn_ball_offset_l {parent=foot_l, loc=<-0.08 0.3776 -0.1451>, rot=<-0.0284 0.0354 -0.7384 -0.6729>}
 dyn_ball_main_l {parent=dyn_ball_offset_l, loc=<7.4143 3.6025 0.2142>, rot=<-0.0026 0.0026 0.0476 -0.9989>}
 ball_l {parent=foot_l, loc=<-6.1599 13.8264 0.6806>, rot=<0.0001 -0.0001 -0.7063 -0.7079>}
 dyn_calf_offset_l {parent=calf_l, loc=<0 0 0>, rot=<0 0 0 1>}
 dyn_kneepad_main_l {parent=dyn_calf_offset_l, loc=<0.0605 0.1198 0.0829>, rot=<-0.0078 -0.004 -0.6902 -0.7235>}
 dyn_kneepad_base_l {parent=dyn_kneepad_main_l, loc=<5.2418 0.3799 0.2493>, rot=<0 0 0.0045 -1>}
 dyn_kneepad_01_l {parent=dyn_kneepad_base_l, loc=<2.1477 6.1079 0.005>, rot=<0 0 -0.8432 -0.5376>}
 dyn_kneepad_02_l {parent=dyn_kneepad_01_l, loc=<-15.7295 4.1015 0>, rot=<-0.1109 0.0176 0.2658 -0.9575>}
 calf_twist_01_l {parent=calf_l, loc=<-35.813 -0.0967 0.0397>, rot=<0.0237 0.0095 -0.0035 -0.9997>}
 dyn_thigh_offset_l {parent=thigh_l, loc=<0 0 0>, rot=<0 0 0 1>}
 dyn_thighpad_l {parent=dyn_thigh_offset_l, loc=<4.6679 -1.6832 -8.2396>, rot=<0.5288 -0.4737 -0.4637 -0.53>}
 thigh_twist_01_l {parent=thigh_l, loc=<-21.6303 0 0>, rot=<0.0013 0 0 -1>}
 thigh_r {parent=pelvis, loc=<-2.5691 -0.0109 7.7501>, rot=<0.0572 -0.0012 0.9982 0.0155>}
 calf_r {parent=thigh_r, loc=<39.0487 0 0>, rot=<0.0006 -0.0161 -0.0391 -0.9991>}
 calf_twist_02_r {parent=calf_r, loc=<18.8064 0.0967 -0.0397>, rot=<0.0237 0.0095 -0.0035 -0.9997>}
 foot_r {parent=calf_r, loc=<42.5612 -0.1077 -0.4292>, rot=<0.0418 -0.0455 0.0343 -0.9975>}
 dyn_heel_offset_r {parent=foot_r, loc=<0.0799 -0.3776 0.145>, rot=<-0.0284 0.0354 -0.7384 -0.6729>}
 dyn_heel_01_r {parent=dyn_heel_offset_r, loc=<5.3597 -7.3285 0.0894>, rot=<0 0 0.0515 -0.9987>}
 dyn_heel_02_r {parent=dyn_heel_01_r, loc=<4.6307 7.0743 -0.2556>, rot=<0.0002 -0.0005 0.5638 -0.8259>}
 dyn_ball_offset_r {parent=foot_r, loc=<0.0799 -0.3776 0.145>, rot=<-0.0284 0.0354 -0.7384 -0.6729>}
 dyn_ball_main_r {parent=dyn_ball_offset_r, loc=<-7.4143 -3.6025 -0.2141>, rot=<-0.004 -0.0254 0.0476 -0.9985>}
 ball_r {parent=foot_r, loc=<6.1598 -13.8264 -0.6806>, rot=<0.0001 -0.0001 -0.7063 -0.7079>}
 dyn_calf_offset_r {parent=calf_r, loc=<-0.0001 0 -0.0001>, rot=<0 0 0 1>}
 dyn_kneepad_main_r {parent=dyn_calf_offset_r, loc=<-0.0604 -0.1198 -0.0829>, rot=<-0.0078 -0.004 -0.6902 -0.7235>}
 dyn_kneepad_base_r {parent=dyn_kneepad_main_r, loc=<-5.2418 -0.3798 -0.2493>, rot=<0 0 0.0045 -1>}
 dyn_kneepad_01_r {parent=dyn_kneepad_base_r, loc=<-2.1477 -6.1079 -0.005>, rot=<0 0 -0.8432 -0.5376>}
 dyn_kneepad_02_r {parent=dyn_kneepad_01_r, loc=<15.7295 -4.1015 0>, rot=<-0.1109 0.0176 0.2658 -0.9575>}
 calf_twist_01_r {parent=calf_r, loc=<35.8132 0.0967 -0.0397>, rot=<0.0237 0.0095 -0.0035 -0.9997>}
 dyn_thigh_offset_r {parent=thigh_r, loc=<0.0002 0 0>, rot=<0 0 0 1>}
 dyn_thighpad_r {parent=dyn_thigh_offset_r, loc=<-4.6679 1.6832 8.2397>, rot=<0.5288 -0.4737 -0.4637 -0.53>}
 thigh_twist_01_r {parent=thigh_r, loc=<21.6303 0 0>, rot=<0.0013 0 0 -1>}
 dyn_flap_rear {parent=pelvis, loc=<7.9691 -14.9077 -0.0047>, rot=<0.2429 -0.97 0 0>}
]
variables [Ads_Float {type=double, default=0}]
