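# Physics asset for the collar of the medium male Armadillo Robot player body, defining bodies for simulated collar movement.
# Front, middle and back collar bones each carry a thin box body, constrained to a tiny capsule on a matching driver bone.
# All three constraints leave swing and twist free, so their stored angle limits do not restrict rotation.
# An unconstrained head capsule provides a volume alongside the flaps, likely for collision.
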PhysicsAsset /Game/Characters/Player/Male/Medium/Bodies/M_MED_Armadillo_Robot/Meshes/M_MED_Armadillo_Robot_Collar_Physics
bodies {
 driver_dyn_collar_bck_01 [{shapeType=Capsule, length=0.505, radius=0.505, rotation=<rot 0 -90 0>}]
 driver_dyn_collar_frt_01 [{shapeType=Capsule, length=0.505, radius=0.505, rotation=<rot 0 -90 0>}]
 driver_dyn_collar_mid_01 [{shapeType=Capsule, length=0.505, radius=0.505, rotation=<rot 0 -90 0>}]
 dyn_collar_bck_01 [{shapeType=Box, extentX=1.05, extentY=16, extentZ=10, center=<6.7061 0.41 1.65>, rotation=<rot 0 -90 0>}]
 dyn_collar_frt_01 [{shapeType=Box, extentX=1.01, extentY=17.51, extentZ=9.01, center=<6.5142 0 -1.5694>, rotation=<rot 0 -89.9999 -1.5281>}]
 dyn_collar_mid_01 [{shapeType=Box, extentX=1.05, extentY=16.51, extentZ=11.01, center=<6.573 0.2 0>, rotation=<rot 2.5 -90 -0.5>}]
 head [{shapeType=Capsule, length=5.5, radius=11.755, center=<1.7017 1.7149 -0.0001>, rotation=<rot 90 -4.7636 71.8597>}]
}
constraints [
 {
  bone1Name=dyn_collar_mid_01
  bone2Name=driver_dyn_collar_mid_01
  swing1LimitDegrees=45
  swing1Motion=Free
  swing2LimitDegrees=45
  swing2Motion=Free
  twistLimitDegrees=45
  twistMotion=Free
 }
 {
  bone1Name=dyn_collar_frt_01
  bone2Name=driver_dyn_collar_frt_01
  swing1LimitDegrees=45
  swing1Motion=Free
  swing2LimitDegrees=45
  swing2Motion=Free
  twistLimitDegrees=45
  twistMotion=Free
 }
 {
  bone1Name=dyn_collar_bck_01
  bone2Name=driver_dyn_collar_bck_01
  swing1LimitDegrees=45
  swing1Motion=Free
  swing2LimitDegrees=45
  swing2Motion=Free
  twistLimitDegrees=45
  twistMotion=Free
 }
]
