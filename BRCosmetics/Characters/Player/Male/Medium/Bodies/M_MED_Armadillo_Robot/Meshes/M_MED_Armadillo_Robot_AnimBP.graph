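# Animation Blueprint for the Armadillo Robot player body cosmetic, adapting its pose and material visibility to gameplay.
# Copies the attached parent's pose through a Control Rig, then adjusts arm, shoulder and backpack bones using aiming, skydiving and hand-twist alphas.
# The event graph sets those alphas and toggles "Hide Element" material parameters for gauntlet pickaxes, replaced hands, emotes, driving and skydiving.
# A collar RigidBody node sits unconnected.
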AnimBlueprint /BRCosmetics/Characters/Player/Male/Medium/Bodies/M_MED_Armadillo_Robot/Meshes/M_MED_Armadillo_Robot_AnimBP
target_skeleton=/Game/Characters/Player/Male/Medium/Bodies/M_MED_Armadillo_Robot/Meshes/M_MED_Armadillo_Robot_Skeleton
graphs [
 AnimGraph {nodes [
   RigidBody() {
    bClampLinearTranslationLimitToRefPose=true
    bOverrideWorldGravity=true
    overridePhysicsAsset=/Game/Characters/Player/Male/Medium/Bodies/M_MED_Armadillo_Robot/Meshes/M_MED_Armadillo_Robot_Collar_Physics
    simulationSpace=BaseBoneSpace
    alphaBoolBlend {blendOption=Linear}
    baseBoneRef {boneName=root}
    componentAppliedLinearAccClamp=<1000 1000 1100>
    componentLinearAccScale=<0.5 0.5 0.5>
    componentLinearVelScale=<0.5 0.5 0.5>
    overrideWorldGravity {z=-980}
   }
   n1 = Get Skydive_Glider_Alpha() -> (Skydive_Glider_Alpha:double)
   n2 = Get Skydive_Float_Alpha() -> (Skydive_Float_Alpha:double)
   n3 = Control Rig!(Source=Copy Pose From Mesh() {bUseAttachedParent=true}, Alpha, Ads_Float:double=Get ads_alpha() -> (ads_alpha:double)) -> (Pose) {
    alphaBoolBlend {blendOption=Linear}
    controlRigAssetReference {blueprintRigClass=&/M_MED_Armadillo_Robot_CtrlRig.M_MED_Armadillo_Robot_CtrlRig_C}
   }
   n4 = "Transform (Modify) Bone - Bone: dyn_lowerarm_dyno_02_r"(
    ComponentPose=Local To Component(LocalPose=@n3)
    Alpha=Get Arm_Rot() -> (Arm_Rot:double)
    Translation="0,0,0"
    Rotation="25,0,0"
    Scale="1,1,1"
   ) {
    rotationMode=BMM_Additive
    rotationSpace=BCS_BoneSpace
    scaleSpace=BCS_BoneSpace
    translationSpace=BCS_BoneSpace
    alphaBoolBlend {blendOption=Linear}
    alphaScaleBiasClamp {bInterpResult=true, interpSpeedDecreasing=50, interpSpeedIncreasing=50}
    boneToModify {boneName=dyn_lowerarm_dyno_02_r}
    rotation {pitch=25}
   }
   n5 = "Transform (Modify) Bone - Bone: dyn_shldr_base_r"(ComponentPose=@n4, Alpha=@n2, Translation="0,5,0", Rotation="0,0,0", Scale="1,1,1") {
    translationMode=BMM_Additive
    translationSpace=BCS_BoneSpace
    alphaBoolBlend {blendOption=Linear}
    boneToModify {boneName=dyn_shldr_base_r}
    translation {y=5}
   }
   n6 = "Transform (Modify) Bone - Bone: dyn_shldr_base_l"(ComponentPose=@n5, Alpha=@n2, Translation="0,-5,0", Rotation="0,0,0", Scale="1,1,1") {
    translationMode=BMM_Additive
    translationSpace=BCS_BoneSpace
    alphaBoolBlend {blendOption=Linear}
    boneToModify {boneName=dyn_shldr_base_l}
    translation {y=-5}
   }
   n7 = "Transform (Modify) Bone - Bone: dyn_shldr_base_r"(ComponentPose=@n6, Alpha=@n1, Translation="0,0,0", Rotation="0,30,0", Scale="1,1,1") {
    rotationMode=BMM_Additive
    rotationSpace=BCS_BoneSpace
    translationMode=BMM_Additive
    translationSpace=BCS_BoneSpace
    alphaBoolBlend {blendOption=Linear}
    boneToModify {boneName=dyn_shldr_base_r}
    rotation {yaw=30}
   }
   n8 = "Transform (Modify) Bone - Bone: dyn_shldr_base_l"(ComponentPose=@n7, Alpha=@n1, Translation="0,0,0", Rotation="0,30,0", Scale="1,1,1") {
    rotationMode=BMM_Additive
    rotationSpace=BCS_BoneSpace
    translationMode=BMM_Additive
    translationSpace=BCS_BoneSpace
    alphaBoolBlend {blendOption=Linear}
    boneToModify {boneName=dyn_shldr_base_l}
    rotation {yaw=30}
   }
   n9 = "Constraint - dyn_backpack (PP)"!(ComponentPose=@n8, Alpha=1, ConstraintWeights_0=0.8734, ConstraintWeights_1:float=0.1266) -> (Pose) {
    alphaBoolBlend {blendOption=Linear}
    boneToModify {boneName=dyn_backpack}
    constraintSetup [{targetBone {boneName=spine_03}}, {targetBone {boneName=spine_05}}]
    constraintWeights [1,1]
   }
   n10 = Get ads_alpha()
   n11 = "Transform (Modify) Bone - Bone: dyn_upperarm_offset_r"(ComponentPose=@n9, Alpha=@n10, Translation="0,0,0", Rotation="0,0,0", Scale="1,0.8,0.95") {
    rotationMode=BMM_Additive
    rotationSpace=BCS_BoneSpace
    scaleMode=BMM_Additive
    scaleSpace=BCS_BoneSpace
    translationMode=BMM_Additive
    translationSpace=BCS_BoneSpace
    alphaBoolBlend {blendOption=Linear}
    alphaScaleBiasClamp {bInterpResult=true, interpSpeedDecreasing=50, interpSpeedIncreasing=50}
    boneToModify {boneName=dyn_upperarm_offset_r}
    scale {y=0.8, z=0.95}
   }
   n12 = "Transform (Modify) Bone - Bone: dyn_shldr_base_r"(ComponentPose=@n11, Alpha=@n10, Translation="0,0,0", Rotation="0,0,0", Scale="0.875,0.875,0.875") {
    rotationMode=BMM_Additive
    rotationSpace=BCS_BoneSpace
    scaleMode=BMM_Additive
    scaleSpace=BCS_BoneSpace
    translationMode=BMM_Additive
    translationSpace=BCS_BoneSpace
    alphaBoolBlend {blendOption=Linear}
    alphaScaleBiasClamp {bInterpResult=true, interpSpeedDecreasing=50, interpSpeedIncreasing=50}
    boneToModify {boneName=dyn_shldr_base_r}
    scale=<0.875 0.875 0.875>
   }
   Output Pose(Result=Component To Local(ComponentPose=@n12)) {layerGroup=None}
  ]}
 EventGraph {nodes [
   Event BlueprintInitializeAnimation()
   > n1 = Cast To PlayerPawn_Athena(Object=TryGetPawnOwner())
   > n2 = Set Pawn Owner(Pawn Owner:Player Pawn Athena Object Reference=@n1) -> (Output_Get:Player Pawn Athena Object Reference)
   > Set Body_Mesh(Body_Mesh=GetSkeletalMeshForPartType(self=@n2))
   n3 = Event OnWeaponChanged()
   n4 = Get Dynamic Material Instance(execute=[@n3.then, @n25.then], SkeletalMeshComponent=GetPartSkeletalMeshComponent())
   > n5 = Sequence()
   n6 = PickaxeItemDefinitionHasAnyMetaTag(Tags=Make Array("[0]"=(TagName="Cosmetics.Pickaxe.Gauntlet")), bRequiresEquipped=true)
   > n7 = Branch(Condition=@n6)
   > n8 = Branch(Condition=Get bIsSkyDiving())
   SetScalarParameterValue(execute=[@n10.then, @n8.then, @n9.else, @n7.else], self=@n4, ParameterName=Hide Element 07, Value=0)
   SetScalarParameterValue(execute=[@n8.else, @n9.then], self=@n4, ParameterName=Hide Element 07, Value=1)
   n9 = Branch(execute=[@n11.then, @n5.then_1], Condition=@n6)
   n10 = Event OnBeginSkydiving()
   n11 = Event OnEndSkydiving()
   Event BlueprintUpdateAnimation()
   > n12 = Sequence!(execute) -> (then_0, then_1, then_2:Exec, then_3:Exec, then_4:Exec, then_5:Exec)
   > n13 = Branch(Condition=Get bIsTargeting())
   > Set ads_alpha(ads_alpha:double=1) -> (Output_Get:double)
   Set ads_alpha(execute=@n13.else, ads_alpha=0)
   Set Arm_Rot(
    execute=@n12.then_3
    Arm_Rot:double=GetAlphaFromReferenceTransformDelta(
     SkeletalMeshComponent=Get Body_Mesh()
     BoneName=hand_r
     BaseName=lowerarm_r
     InputRange=(X=53,Y=15)
     OutputRange=(X=1,Y=0)
     TwistAxis="0,0,1"
    )
   ) -> (Output_Get:double)
   n14 = Get bIsSkydiveFloating()
   n15 = Get bIsSkydiveDiveMode()
   Set Skydive_Float_Alpha(execute=@n12.then_4, Skydive_Float_Alpha:double=SelectFloat(A=1, B=0, bPickA=OR Boolean(A=@n14, B=@n15))) -> (Output_Get:double)
   Set Skydive_Glider_Alpha(execute=@n12.then_5, Skydive_Glider_Alpha:double=SelectFloat(A=1, B=0, bPickA=Get bIsSkyDiving())) -> (Output_Get:double)
   n16 = Cast To FortPlayerPawn(execute=@n12.then_1, Object=GetOwningActor())
   > Set As Fort Player Pawn(As Fort Player Pawn:Fort Player Pawn Object Reference=@n16) -> (Output_Get:Fort Player Pawn Object Reference)
   > n17 = Get!(execute:Exec) -> (then:Exec, else:Exec, As Fort Player Pawn:Fort Player Pawn Object Reference)
   > Set is_wearing_westsausage(is_wearing_westsausage:bool=HasMatchingGameplayTag!(self:Gameplay Tag Asset Interface Interface=@n17, TagToCheck:Gameplay Tag Structure=(TagName="Granted.Athena.HandsAreReplaced")) -> (ReturnValue:bool)) -> (Output_Get:bool)
   n18 = Get Dynamic Material Instance(execute=@n12.then_2, SkeletalMeshComponent=GetPartSkeletalMeshComponent())
   > n19 = Branch(Condition=Get is_wearing_westsausage() -> (is_wearing_westsausage:bool))
   > n20 = Branch(Condition=Get is_performing_emote() -> (is_performing_emote:bool))
   SetScalarParameterValue(execute=[@n24.then, @n23.then, @n20.then, @n22.then, @n19.else, @n21.then], self=@n18, ParameterName=Hide Element 09, Value=0)
   n21 = Branch(execute=@n20.else, Condition=Get bIsValetDriver() -> (bIsValetDriver:bool))
   n22 = Branch(execute=@n21.else, Condition=Get bIsMeatballDriver() -> (bIsMeatballDriver:bool))
   n23 = Branch(execute=@n22.else, Condition=Get bIsAntelopeVehicleDriver() -> (bIsAntelopeVehicleDriver:bool))
   n24 = Branch(execute=@n23.else, Condition=Get bIsSkyDiving())
   SetScalarParameterValue(execute=@n24.else, self=@n18, ParameterName=Hide Element 09, Value=1)
   Event OnBeginPlayingEmote() {is=title K2Node_Event}
   > Set is_performing_emote(is_performing_emote:bool=true) -> (Output_Get:bool)
   n25 = Event OnCharacterCustomizationLoaded()
   Event OnEndPlayingEmote() {is=title K2Node_Event}
   > Set is_performing_emote(is_performing_emote=false)
  ]}
]
variables [
 ads_alpha {type=number}
 Arm_Rot {type=number}
 As Fort Player Pawn {type=FortPlayerPawn}
 Body_Mesh {type=SkeletalMeshComponent}
 is_performing_emote {type=boolean}
 is_wearing_westsausage {type=boolean}
 Pawn Owner {type=PlayerPawn_Athena_C}
 Skydive_Float_Alpha {type=number}
 Skydive_Glider_Alpha {type=number}
]
